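# Physics asset for the DangerMojo axe pickaxe, a two-handed melee weapon cosmetic, giving its mesh simple collision.
# It has a single body on the Melee_TwoHanded_Handle bone, shaped as one capsule offset and rotated from that bone.
# No constraints exist, so the weapon acts as one rigid piece.
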PhysicsAsset /BRCosmetics/Weapons/FORT_Melee/Pickaxe_DangerMojo/Meshes/DangerMojo_Axe_PhysicsAsset
bodies {Melee_TwoHanded_Handle [{shapeType=Capsule, length=45.2884, radius=4.8361, center=<0.5732 -7.9398 58.8867>, rotation=<rot -0.0134 -90 1.5246>}]}
constraints []
